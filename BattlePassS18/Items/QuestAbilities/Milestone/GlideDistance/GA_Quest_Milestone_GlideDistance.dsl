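; Quest milestone gameplay ability that tracks horizontal distance a player travels while parachute gliding and reports it as quest progress.
; A randomized looping timer checks the pawn's skydive and parachute tags, resetting the start point during skydive and adding floored XY meters during parachute.
; It sends a custom stat event whenever the total exceeds 100 meters and ends when neither tag matches.
; Ending clears the timer and flushes any remainder of at least 5 meters.
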(event Ability|EventActivateAbilityFromEvent
  (params (EventData :type Gameplay%20Event%20Data%20Structure%20%28by%20ref%29))
  ;[1 1 1] Activate ability
  (|CallParentFunction EventData)
  (bind _output_get (Variables|GAVars|Data|SetPreviousLocation (Transformation|GetActorLocation (Variables|GAVars|Data|GetOwningPlayerPawn))))
  (Variables|GAVars|Data|SetCurrentLocation _output_get)
  (bind _output_get_2 (Variables|GAVars|Setup|SetInterval (Math|Random|RandomFloatinRange 1.7 2.5)))
  (bind _returnvalue (Utilities|Time|SetTimerbyEvent (delegate Custom|SendDistanceUpdate) _output_get_2 true))
  (Variables|GAVars|Data|SetTImerHandle _returnvalue))

(event Custom|SendDistanceUpdate
  (bind _owningplayerpawn (Variables|GAVars|Data|GetOwningPlayerPawn))
  (bind _t_parachute (Variables|GAVars|Setup|GetT_Parachute))
  (bind _t_skydive (Variables|GAVars|Setup|GetT_Skydive))
  (bind (_case_0_true _case_1_true) (GameplayTags|TagContainer|CompareTagAssetInterfacetoOtherTags _t_skydive _t_parachute _owningplayerpawn))
  (bind (_previouslocation_x _previouslocation_y) (Math|Vector|BreakVector (Variables|GetPreviousLocation_X)))
  (bind (_currentlocation_x _currentlocation_y) (Math|Vector|BreakVector (Variables|GetCurrentLocation_X)))
  (bind _returnvalue (Math|Vector|vector-vector (Math|Vector|MakeVector _currentlocation_x _currentlocation_y) (Math|Vector|MakeVector _previouslocation_x _previouslocation_y)))
  (bind _returnvalue_2 (Math|Vector|VectorLength _returnvalue))
  (bind _returnvalue_3 (/ _returnvalue_2 1e+02))
  ;[1 1 1] Send Distance Update
  (Variables|GAVars|Data|SetCurrentLocation (Transformation|GetActorLocation _owningplayerpawn))
  (if _case_0_true
    (Variables|GAVars|Data|SetPreviousLocation (Variables|GAVars|Data|GetCurrentLocation))
    (Variables|Default|SetPreviously%20Gliding true)
    (elif _case_1_true
      (if (Variables|Default|GetPreviously%20Gliding)
        ;;[1 1 1] Cache distance traveled (horizontal only)
        (Variables|GAVars|Data|SetPreviousLocation (Variables|GAVars|Data|GetCurrentLocation))
        (Variables|Default|SetPreviously%20Gliding false)
        (else
          _))
      (Development|PrintString (Utilities|String|ToString(Float) _returnvalue_3))
      (Variables|GAVars|Data|SetMetersTraveled (+ (Math|Float|Floor _returnvalue_3) (Variables|GAVars|Data|GetMetersTraveled)))
      (Variables|GAVars|Data|SetPreviousLocation (Variables|GAVars|Data|GetCurrentLocation)) ; Reset location
      ;;[1 1 1] Send Quest Progress
      (if (> (Variables|GAVars|Data|GetMetersTraveled) 100)
        (Quest|SendComplexCustomStatEvent (Variables|GAVars|Data|GetOwningPlayerQuestManager) 0 (Variables|GAVars|Setup|GetT_Quest) "(GameplayTags=)" (Variables|GAVars|Data|GetMetersTraveled))
        (Variables|GAVars|Data|SetMetersTraveled 0))
      (else
        (Ability|EndAbility)))))

(event Custom|Final%20Distance%20Update
  ;[1 1 1] Did the player move? If So send final update
  (if (>= (Variables|GAVars|Data|GetMetersTraveled) 5)
    ;[1 1 1] Send Distance Update
    ;;[1 1 1] Send Quest Progress
    (Quest|SendComplexCustomStatEvent (Variables|GAVars|Data|GetOwningPlayerQuestManager) 0 (Variables|GAVars|Setup|GetT_Quest) "(GameplayTags=)" (Variables|GAVars|Data|GetMetersTraveled))
    (Variables|GAVars|Data|SetMetersTraveled 0)))

(event Ability|EventOnEndAbility
  (params (bWasCancelled :type Boolean))
  (bind _timerhandle (Variables|GAVars|Data|GetTImerHandle))
  ;[1 1 1] On End Ability
  (|CallParentFunction bWasCancelled)
  (if (Utilities|Time|IsValidTimerHandle _timerhandle)
    (Utilities|Time|ClearandInvalidateTimerbyHandle _timerhandle)
    (else
      _))
  (CallFunction|FinalDistanceUpdate))
